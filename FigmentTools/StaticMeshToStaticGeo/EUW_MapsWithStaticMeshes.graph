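# Editor utility widget that finds other maps, outside the current level, containing terrain-style static mesh actors.
# Find scans actor descriptors for FortStaticMeshActors whose package lacks the current level name and whose name matches MeshNameFilter keywords like cliff or cave.
# Their existing owning assets, resolved from actor paths, fill the Levels map shown in a details view.
# A SaveGame slot persists the map, restored on construct, with manual Load and Save buttons.
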EditorUtilityWidgetBlueprint /FigmentTools/StaticMeshToStaticGeo/EUW_MapsWithStaticMeshes
graphs [
 Contains {nodes [
   n1 = Contains!() -> (then, Substring:String) {is=title K2Node_FunctionEntry}
   > n2 = Sequence()
   > n3 = For Each Loop with Break!(Exec:Exec, Array:Array of Strings=Get MeshNameFilter() -> (MeshNameFilter:Array of Strings), Break:Exec=@n4.then) -> (LoopBody:Exec, Array Element:String, Array Index:int, Completed:Exec)
   > Branch(Condition=Contains(SearchIn=@n3, Substring=@n1) {is=title K2Node_CallFunction})
   > n4 = Set Found(Found:bool=true) -> (Output_Get:bool)
   Return Node!(execute=@n2.then_1, Return:bool=Get Found() -> (Found:bool))
  ]}
 EventGraph {nodes [
   "On Clicked (EditorUtilityButton_0)"() {is=title K2Node_ComponentBoundEvent}
   > Clear!(execute, TargetMap:Map of Soft Object References to Strings=Get Levels() -> (Levels:Map of Soft Object References to Strings)) -> (then) {is=title K2Node_CallFunction}
   > n1 = GetCurrentLevelName!(execute, WorldContextObject:Object Reference, bRemovePrefixString:bool=true) -> (then, ReturnValue:String)
   > Set CurrentLevelName(CurrentLevelName:String=@n1) -> (Output_Get:String)
   > n2 = GetActorDescs!(execute) -> (then, OutActorDescs:Array of Actor Desc Structures, ReturnValue:bool)
   > n3 = For Each Loop(Array:Array of Actor Desc Structures=@n2) -> (Array Element:Actor Desc Structure)
   n4 = Break Actor Desc!(ActorDesc:"Actor Desc Structure (by ref)"=@n3) -> (
    Guid:Guid Structure
    NativeClass:Object Class Reference
    Class:Soft Object Path Structure
    Name:Name
    Label:Name
    Bounds:Box Structure
    RuntimeGrid:Name
    bIsSpatiallyLoaded:bool=~
    bActorIsEditorOnly:bool=~
    ActorPackage:Name
    ActorPath:Name
    FolderPath:Name
    DataLayerAssets:Array of Soft Object Path Structures
   ) {is=title K2Node_BreakStruct}
   > Branch(Condition="Equal (Class)"(A=@n4.NativeClass, B=/Script/FortniteGame.FortStaticMeshActor))
   > n5 = Branch(Condition=Contains(SearchIn="To String (Name)"(InName=@n4.ActorPackage), Substring=Get CurrentLevelName() -> (CurrentLevelName:String), bUseCase=true))
   Branch(execute=@n5.else, Condition=Contains!(Substring="To String (Name)"(InName=@n4.Name)) -> (Return:bool))
   n6 = Split(SourceString="To String (Name)"(InName=@n4.ActorPath), InStr=., SearchDir=FromStart)
   > n7 = DoesAssetExist(AssetPath=@n6)
   > Branch(Condition=@n7)
   > n8 = FindAssetData!(execute, AssetPath:String=@n6) -> (then, ReturnValue:Asset Data Structure)
   n9 = Get Levels()
   > Add!(
    execute
    TargetMap:Map of Soft Object References to Strings=@n9
    Key:"Soft Object Reference (by ref)"=Make Soft Reference!(Input:Object Reference=GetAsset(InAssetData=@n8)) -> (Output:Soft Object Reference)
    Value:"String (by ref)"=@n6
   ) -> (then) {is=title K2Node_CallFunction}
   n10 = Get SavedData() -> (SavedData:Save Maps With Static Meshes Object Reference)
   > Set SaveData!(execute, SaveData:Map of Soft Object References to Strings=@n9, self:Save Maps With Static Meshes Object Reference=@n10) -> (then, Output_Get:Map of Soft Object References to Strings)
   > SaveGameToSlot!(execute, SaveGameObject:Save Game Object Reference=@n10, SlotName:String=MySave, UserIndex:int=0) -> (then, ReturnValue:bool)
   "On Clicked (EditorUtilityButton)"() {is=title K2Node_ComponentBoundEvent}
   > n11 = LoadGameFromSlot!(execute, SlotName:String=MySave, UserIndex:int=0) -> (then, ReturnValue:Save Game Object Reference)
   > n12 = Cast To Save_MapsWithStaticMeshes!(execute, Object=@n11) -> (then, CastFailed, AsSave Maps with Static Meshes:Save Maps With Static Meshes Object Reference) {is=title K2Node_DynamicCast}
   > n13 = Set SavedData(SavedData:Save Maps With Static Meshes Object Reference=@n12) -> (Output_Get:Save Maps With Static Meshes Object Reference)
   > Set Levels(Levels:Map of Soft Object References to Strings=Get SaveData!(self:Save Maps With Static Meshes Object Reference=@n13) -> (SaveData:Map of Soft Object References to Strings)) -> (Output_Get:Map of Soft Object References to Strings)
   "On Clicked (EditorUtilityButton_1)"() {is=title K2Node_ComponentBoundEvent}
   n14 = Get SavedData()
   > Set SaveData(SaveData=Get Levels(), self=@n14)
   > SaveGameToSlot(SaveGameObject=@n14, SlotName=MySave, UserIndex=0)
   Event PreConstruct()
   > SetObject(self=Get DetailsView_41() -> (DetailsView_41:Details View Object Reference), NewObject=Self-Reference())
   Event Construct()
   > n15 = CreateSaveGameObject!(execute, SaveGameClass:Save Game Class Reference=&/Save_MapsWithStaticMeshes.Save_MapsWithStaticMeshes_C) -> (then, ReturnValue:Save Maps With Static Meshes Object Reference)
   > Set SavedData(SavedData=@n15)
   > n16 = LoadGameFromSlot(SlotName=MySave, UserIndex=0)
   > n17 = Cast To Save_MapsWithStaticMeshes(Object=@n16)
   > n18 = Set SavedData(SavedData=@n17)
   > Set Levels(Levels=Get SaveData(self=@n18))
   MakeSoftObjectPath!(PathString:String=@n6) -> (ReturnValue:Soft Object Path Structure)
   Event Tick()
  ]}
]
variables [
 CurrentLevelName {type=string, default=""}
 Levels {default {}}
 MeshNameFilter {type="array<string>", default [cliff,elevation,cave,shoreline,slope]}
 SavedData {type=Save_MapsWithStaticMeshes_C, default=null}
]
widgets {
 info {rootWidgetClass=&CanvasPanel, widgetCount=8}
 widgets [
  CanvasPanel_21 {widgetClassPath=&CanvasPanel, widget=CanvasPanel_21}
  DetailsView_41 {
   widgetClassPath=&DetailsView
   parent=CanvasPanel_21
   widget=DetailsView_41
   properties {propertiesToShow [Levels]}
   slot {
    class=CanvasPanelSlot
    properties {layoutData {alignment=<0 0>, anchors {maximum=<1 1>, minimum=<0 0>}, offsets=<ltrb 28 84 60.1881 38.1521>}}
   }
  }
  EditorUtilityButton_0 {
   widgetClassPath=&EditorUtilityButton
   parent=CanvasPanel_21
   widget=EditorUtilityButton_0
   slot {class=CanvasPanelSlot, properties {layoutData {alignment=<0 0>, anchors {maximum=<0 0>, minimum=<0 0>}, offsets=<ltrb 24 20 100 40>}}}
  }
  TextBlock_17 {widgetClassPath=&TextBlock, parent=EditorUtilityButton_0, widget=TextBlock_17, properties {text=Find}, slot {class=ButtonSlot}}
  EditorUtilityButton {
   widgetClassPath=&EditorUtilityButton
   parent=CanvasPanel_21
   widget=EditorUtilityButton
   slot {class=CanvasPanelSlot, properties {layoutData {alignment=<0 0>, anchors {maximum=<0 0>, minimum=<0 0>}, offsets=<ltrb 144 20 100 40>}}}
  }
  TextBlock {widgetClassPath=&TextBlock, parent=EditorUtilityButton, widget=TextBlock, properties {text=Load}, slot {class=ButtonSlot}}
  EditorUtilityButton_1 {
   widgetClassPath=&EditorUtilityButton
   parent=CanvasPanel_21
   widget=EditorUtilityButton_1
   slot {class=CanvasPanelSlot, properties {layoutData {alignment=<0 0>, anchors {maximum=<0 0>, minimum=<0 0>}, offsets=<ltrb 260 20 100 40>}}}
  }
  TextBlock_1 {widgetClassPath=&TextBlock, parent=EditorUtilityButton_1, widget=TextBlock_1, properties {text=Save}, slot {class=ButtonSlot}}
 ]
}
