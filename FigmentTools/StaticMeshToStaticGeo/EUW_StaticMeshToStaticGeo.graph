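# Editor Utility Widget for converting Fortnite static mesh actors in a level, filtered by mesh name keywords like cliff or cave, into building Blueprint actors.
# Buttons pin or unpin matching world actor descs, gather unique meshes, and create child Blueprints (S_ renamed BP_) with the mesh assigned.
# Replace Actors swaps each actor for its Blueprint and sets TextureData from a texture data asset whose path contains the material name.
# A self-bound DetailsView exposes the settings.
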EditorUtilityWidgetBlueprint /FigmentTools/StaticMeshToStaticGeo/EUW_StaticMeshToStaticGeo
graphs [
 Contains {nodes [
   n1 = Contains!() -> (then, Substring:String) {is=title K2Node_FunctionEntry}
   n2 = Get MeshNameFilter() -> (MeshNameFilter:Array of Strings)
   > n3 = For Each Loop(Array=@n2)
   Return Node!(execute=@n3.Completed, Return:bool=Get Found() -> (Found:bool))
   > Branch(Condition=Contains(SearchIn=@n1, Substring=@n3) {is=title K2Node_CallFunction})
   > Set Found(Found:bool=true) -> (Output_Get:bool)
   n4 = For Each Loop with Break!(Exec:Exec, Array:Array of Strings=@n2, Break:Exec=@n6.then) -> (LoopBody:Exec, Array Element:String, Array Index:int, Completed:Exec)
   n5 = Contains(SearchIn=@n4, Substring=@n1)
   > Branch(Condition=@n5)
   > n6 = Set Found(Found=true)
  ]}
 EventGraph {nodes [
   n1 = GetAllActorsOfClass!(execute, WorldContextObject:Object Reference, ActorClass:Actor Class Reference=/Script/FortniteGame.FortStaticMeshActor) -> (then, OutActors:Array of Fort Static Mesh Actor Object References)
   > n2 = Set AllStaticMeshes(AllStaticMeshes:Array of Static Mesh Actor Object References=@n1) -> (Output_Get:Array of Static Mesh Actor Object References)
   n3 = For Each Loop(
    Exec=[@n44.else, @n2.then]
    Array:Array of Static Mesh Actor Object References=Get AllStaticMeshes() -> (AllStaticMeshes:Array of Static Mesh Actor Object References)
   ) -> (Array Element:Static Mesh Actor Object Reference)
   n4 = Get FilterByStaticMeshAssetName() -> (FilterByStaticMeshAssetName:bool)
   n5 = Get StaticMesh(self=Get StaticMeshComponent(self=@n3))
   > PrintString(InString=GetDisplayName(Object=Select(Option 0:Object Reference=@n3, Option 1:Object Reference=@n5, Index=@n4) -> (ReturnValue:Object Reference)))
   n6 = GetBlueprintAsset()
   > n7 = GetEditorProperty!(execute:Exec, Object:Object Reference=@n6, PropertyName:Name=TextureData) -> (then:Exec, PropertyValue:Array of Building Texture Data Object References, ReturnValue:bool)
   > SetEditorProperty(
    Object=Reroute Node(InputPin:Object Reference) -> (OutputPin:Object Reference)
    PropertyName=TextureData
    PropertyValue:Array of Building Texture Data Object References=@n7
   )
   n8 = GetAllActorsOfClass(ActorClass=&FortStaticMeshActor)
   > n9 = For Each Loop(Array:Array of Fort Static Mesh Actor Object References=@n8) -> (Array Element:Fort Static Mesh Actor Object Reference)
   > n10 = Set TD_Path(TD_Path:String=MatchCurrentMaterialWithTD!(Target:Static Mesh Actor Object Reference=@n9) -> (Texture Data:String)) -> (Output_Get:String)
   > n11 = DoesAssetExist(AssetPath=@n10)
   > Set ReplaceTD(ReplaceTD:bool=@n11) -> (Output_Get:bool)
   n12 = GetObjectName(Object=Get StaticMesh(self=Get StaticMeshComponent(self=@n9)))
   > Branch(Condition=Contains(SearchIn=@n12, Substring=Get MeshNameFilter()))
   > SetSelectedLevelActors(self=Get EditorActorSubsystem(), ActorsToSelect=Make Array("[0]":Actor Object Reference=@n9) -> (Array:Array of Actor Object References))
   n13 = Append(A=Get Path() -> (Path:String), B=Replace(SourceString=@n12, From=S_, To=BP_))
   > n14 = DoesAssetExist(AssetPath=@n13)
   > n15 = Branch(Condition=@n14)
   > ReplaceSelectedActors!(execute, InAssetPath:String=@n13) -> (then)
   > n16 = GetSelectedActors()
   > n17 = DoesAssetExist(AssetPath=Get TD_Path() -> (TD_Path:String))
   > n18 = Branch(Condition=@n17)
   > n19 = FindAssetData!(execute, AssetPath:String=Get TD_Path()) -> (then, ReturnValue:Asset Data Structure)
   > n20 = Cast To BuildingTextureData!(execute, Object=GetAsset(InAssetData=@n19)) -> (then, CastFailed, AsBuilding Texture Data:Building Texture Data Object Reference) {is=title K2Node_DynamicCast}
   n21 = "Get (a copy)"(Array:Array of Actor Object References=@n16) -> (Output:Actor Object Reference)
   > SetTextureData!(execute, self, Actor:Actor Object Reference=@n21, TextureData:Building Texture Data Object Reference=@n20, Index:int=0) -> (then)
   PrintString(execute=@n20.CastFailed, InString=Asset is not Building Texture Data)
   ResetEditorProperty!(execute=@n18.else, Object:Object Reference=@n21, PropertyName:Name=TextureData, ChangeNotifyMode:EPropertyAccessChangeNotifyMode Enum) -> (then, ReturnValue:bool)
   PrintString(execute=@n15.else, InString=Append(A="Asset not found: ", B=@n13))
   "On Clicked (EditorUtilityButton_24)"() {is=title K2Node_ComponentBoundEvent}
   > n22 = GetCurrentLevelName!(execute, WorldContextObject:Object Reference, bRemovePrefixString:bool=true) -> (then, ReturnValue:String)
   > Set LevelName(LevelName:String=@n22) -> (Output_Get:String)
   > n23 = GetActorDescs!(execute) -> (then, OutActorDescs:Array of Actor Desc Structures, ReturnValue:bool)
   > n24 = For Each Loop(Array:Array of Actor Desc Structures=@n23) -> (Array Element:Actor Desc Structure)
   n25 = Break Actor Desc!(ActorDesc:"Actor Desc Structure (by ref)"=@n24) -> (
    Guid:Guid Structure
    NativeClass:Object Class Reference
    Class:Soft Object Path Structure
    Name:Name
    Label:Name
    Bounds:Box Structure
    RuntimeGrid:Name
    bIsSpatiallyLoaded:bool=~
    bActorIsEditorOnly:bool=~
    ActorPackage:Name
    ActorPath:Name
    FolderPath:Name
    DataLayerAssets:Array of Soft Object Path Structures
   ) {is=title K2Node_BreakStruct}
   > Branch(Condition=OR Boolean(
     A="Equal (Class)"(A=@n25.NativeClass, B=&FortStaticMeshActor)
     B="Equal (Class)"(A=@n25.NativeClass, B=&StaticMeshActor)
    ))
   > Branch(Condition=Contains(SearchIn="To String (Name)"(InName=@n25.ActorPackage), Substring=Get LevelName() -> (LevelName:String)))
   > Add Unique(TargetArray:Array of Guid Structures=Get Guids() -> (Guids:Array of Guid Structures), NewItem:"Guid Structure (by ref)"=@n25)
   "On Clicked (EditorUtilityButton)"() {is=title K2Node_ComponentBoundEvent}
   > PinActors!(execute, InActorsToPin:Array of Guid Structures=Get Guids()) -> (then)
   "On Clicked (EditorUtilityButton_1)"() {is=title K2Node_ComponentBoundEvent}
   > UnpinActors!(execute, InActorsToUnpin:Array of Guid Structures=Get Guids()) -> (then)
   "On Clicked (EditorUtilityButton_2)"() {is=title K2Node_ComponentBoundEvent}
   > n26 = GetAllActorsOfClass(ActorClass=&FortStaticMeshActor)
   > n27 = For Each Loop(Array:Array of Fort Static Mesh Actor Object References=@n26) -> (Array Element:Fort Static Mesh Actor Object Reference)
   n28 = For Each Loop(
    Exec=@n27.Completed
    Array:Array of Static Mesh Object References=Get UniqueMeshes() -> (UniqueMeshes:Array of Static Mesh Object References)
   ) -> (Array Element:Static Mesh Object Reference)
   n29 = Get Target Array() -> (Target Array:Array of Strings)
   SortStringArray!(execute=@n28.Completed, TargetArray:Array of Strings=@n29, bStableSort:bool=true, SortOrder:EArraySortOrder Enum=Ascending) -> (then)
   Add(execute=@n28.LoopBody, TargetArray=@n29, NewItem=GetObjectName(Object=@n28)) {is=title K2Node_CallArrayFunction}
   n30 = Get StaticMesh(self=Get StaticMeshComponent(self=@n27))
   Branch(execute=@n27.LoopBody, Condition=Contains!(Substring=GetObjectName(Object=@n30)) -> (Return:bool))
   > Add Unique(TargetArray:Array of Static Mesh Object References=Get UniqueMeshes(), NewItem:"Static Mesh Object Reference (by ref)"=@n30)
   "On Clicked (EditorUtilityButton_3)"() {is=title K2Node_ComponentBoundEvent}
   > n31 = GetAllActorsOfClass(ActorClass=&FortStaticMeshActor)
   > n32 = For Each Loop(Array:Array of Fort Static Mesh Actor Object References=@n31) -> (Array Element:Fort Static Mesh Actor Object Reference)
   > ReplaceStuff!(execute, self, Target:Static Mesh Actor Object Reference=@n32) -> (then)
   "On Clicked (EditorUtilityButton_4)"() {is=title K2Node_ComponentBoundEvent}
   > n33 = For Each Loop(Array:Array of Static Mesh Object References=Get UniqueMeshes()) -> (Array Element:Static Mesh Object Reference)
   n34 = Get Class() -> (Class:Object Class Reference) {is=title K2Node_VariableGet}
   > Branch(Condition=IsValidClass!(Class:Object Class Reference=@n34) -> (ReturnValue:bool))
   n35 = Append(A=Get Path(), B=Replace(SourceString=GetObjectName(Object=@n33), From=S_, To=BP_))
   > n36 = DoesAssetExist(AssetPath=@n35)
   > n37 = Branch(Condition=@n36)
   n38 = CreateChildAsset!(execute=@n37.else, Path:String=@n35, Class:Object Class Reference=@n34) -> (then, OutBlueprint:Blueprint Object Reference, bSuccess:bool, ReturnValue:Object Reference)
   > n39 = Gather Subobject Data For Blueprint(self=Get SubobjectDataSubsystem(), Context=@n38)
   > n40 = For Each Loop(Array:Array of Subobject Data Handle Structures=@n39) -> (Array Element:Subobject Data Handle Structure)
   > SetEditorProperty(
    Object=GetObject!(Data:"Subobject Data Structure (by ref)"=GetData(DataHandle=@n40), bEvenIfPendingKill:bool=false) -> (ReturnValue:Object Reference)
    PropertyName=StaticMesh
    PropertyValue:Static Mesh Object Reference=@n33
   )
   "On Clicked (EditorUtilityButton_5)"() {is=title K2Node_ComponentBoundEvent}
   > n41 = GetSelectedActors()
   > n42 = For Each Loop(Array:Array of Actor Object References=@n41) -> (Array Element:Actor Object Reference)
   > SetTextureData(
    Actor=@n42
    TextureData=Get New Building Texture Data() -> (New Building Texture Data:Building Texture Data Object Reference)
    Index=Get Index() -> (Index:int=0)
   )
   n43 = Cast To BuildingProp!(execute, Object:Wildcard) -> (then, CastFailed, AsBuilding Prop:Building Prop Object Reference) {is=title K2Node_DynamicCast}
   > ReadTextureData!(execute, self:Building Static Mesh Object Reference=@n43) -> (then, ReturnValue:Array of Texture 2D Object References)
   Event PreConstruct()
   > SetObject(self=Get DetailsView_0() -> (DetailsView_0:Details View Object Reference), NewObject=Self-Reference())
   n44 = Branch(Condition=Get FilterByStaticMeshAssetName())
   GetBuildingTextureData!(execute, self:Building Static Mesh Object Reference=@n43) -> (then, OutTextureDataArray:Array of Building Texture Data Object References)
   Cast To BuildingActor!(execute, Object:Wildcard) -> (then, CastFailed, AsBuilding Actor:Building Actor Object Reference) {is=title K2Node_DynamicCast}
   Event Construct()
   Event Tick()
  ]}
 MatchCurrentMaterialWithTD {nodes [
   n1 = MatchCurrentMaterialWithTD!() -> (then, Target:Static Mesh Actor Object Reference) {is=title K2Node_FunctionEntry}
   > Is Valid(InputObject=@n1)
   > n2 = ListAssets!(execute, DirectoryPath:String=Get TextureDataDirectory() -> (TextureDataDirectory:String), bRecursive:bool=true, bIncludeFolder:bool=false) -> (then, ReturnValue:Array of Strings)
   > n3 = For Each Loop with Break(Array=@n2, Break=@n5.then)
   Return Node!(execute=@n3.Completed, Texture Data:String=Get Asset() -> (Asset:String))
   n4 = GetMaterial!(self:Primitive Component Object Reference=Get StaticMeshComponent(self=@n1), ElementIndex:int=0) -> (ReturnValue:Material Interface Object Reference)
   > Branch(Condition=Contains!(SearchIn:String=@n3, Substring=GetObjectName(Object=@n4), bUseCase:bool, bSearchFromEnd:bool) -> (ReturnValue:bool))
   > n5 = Set Asset(Asset:String=@n3) -> (Output_Get:String)
  ]}
 ReplaceStuff {nodes [
   n1 = ReplaceStuff!() -> (then, Target:Static Mesh Actor Object Reference) {is=title K2Node_FunctionEntry}
   > n2 = Set Actor(Actor:Static Mesh Actor Object Reference=@n1) -> (Output_Get:Static Mesh Actor Object Reference)
   n3 = Get StaticMesh(self=Get StaticMeshComponent(self=@n2))
   > Branch(Condition=Contains!(Substring=GetObjectName(Object=@n3)) -> (Return:bool))
   > n4 = Set TD Path(TD Path:String=MatchCurrentMaterialWithTD(Target=Get Actor() -> (Actor:Static Mesh Actor Object Reference)) {is=title K2Node_CallFunction}) -> (Output_Get:String)
   > n5 = DoesAssetExist(AssetPath=@n4)
   > Set DoesMaterialHaveTD(DoesMaterialHaveTD:bool=@n5) -> (Output_Get:bool)
   > n6 = FindAssetData(AssetPath=Get TD Path() -> (TD Path:String))
   > n7 = Cast To BuildingTextureData(Object=GetAsset(InAssetData=@n6))
   > Set TD(TD:Building Texture Data Object Reference=@n7) -> (Output_Get:Building Texture Data Object Reference)
   > SetSelectedLevelActors(self=Get EditorActorSubsystem(), ActorsToSelect=Make Array("[0]"=Get Actor()))
   n8 = Get StaticMesh(self=Get StaticMeshComponent(self=Get Actor()))
   n9 = Append(A=Get Path(), B=Replace(SourceString=GetObjectName(Object=@n8), From=S_, To=BP_))
   > n10 = DoesAssetExist(AssetPath=@n9)
   > n11 = Branch(Condition=@n10)
   > ReplaceSelectedActors(InAssetPath=@n9)
   > n12 = GetSelectedActors()
   > n13 = DoesAssetExist(AssetPath=Get TD Path())
   > n14 = Branch(Condition=@n13)
   > SetTextureData(Actor="Get (a copy)"(Array=@n12), TextureData=Get TD() -> (TD:Building Texture Data Object Reference), Index=0)
   PrintString(execute=@n14.else, InString=Something Wwent wrong)
   PrintString(execute=@n11.else, InString=Append(A="Asset not found: ", B=@n9))
   > Add Unique(
    TargetArray:Array of Actor Object References=Get NoAssetAssociatedList() -> (NoAssetAssociatedList:Array of Actor Object References)
    NewItem:"Actor Object Reference (by ref)"=Get Actor()
   )
   PrintString(execute=@n7.CastFailed, InString=Asset is not Building Texture Data)
   > Add Unique(TargetArray=Get NoMaterialAssociatedList() -> (NoMaterialAssociatedList:Array of Actor Object References), NewItem=Get Actor())
  ]}
 SetTextureData {nodes [
   n1 = SetTextureData!() -> (then, Actor:Actor Object Reference, TextureData:Building Texture Data Object Reference, Index:int) {is=title K2Node_FunctionEntry}
   > n2 = GetEditorProperty(Object=@n1, PropertyName=TextureData)
   n3 = "Get (a copy)"(Array:Array of Building Texture Data Object References=@n2) -> (Output:Building Texture Data Object Reference)
   n4 = "Get (a copy)"(Array=@n2, Dimension 1=2)
   n5 = "Get (a copy)"(Array=@n2, Dimension 1=1)
   n6 = "Get (a copy)"(Array=@n2, Dimension 1=1)
   n7 = "Get (a copy)"(Array=@n2, Dimension 1=2)
   n8 = "Get (a copy)"(Array=@n2, Dimension 1=3)
   n9 = "Get (a copy)"(Array=@n2, Dimension 1=3)
   n10 = "Get (a copy)"(Array=@n2)
   n11 = "Get (a copy)"(Array=@n2, Dimension 1=2)
   n12 = "Get (a copy)"(Array=@n2, Dimension 1=1)
   n13 = "Get (a copy)"(Array=@n2, Dimension 1=3)
   n14 = "Get (a copy)"(Array=@n2)
   > SetEditorProperty(
    Object=@n1
    PropertyName=TextureData
    PropertyValue:Array of Building Texture Data Object References=Select!(
     Option 0:Array of Building Texture Data Object References=Make Array!(
      "[0]":Building Texture Data Object Reference=@n1.TextureData
      "[1]":Building Texture Data Object Reference=@n6
      "[2]":Building Texture Data Object Reference=@n7
      "[3]":Building Texture Data Object Reference=@n8
     ) -> (Array:Array of Building Texture Data Object References)
     Option 1:Array of Building Texture Data Object References=Make Array("[0]"=@n10, "[1]"=@n1.TextureData, "[2]"=@n11, "[3]"=@n9)
     Option 2:Array of Building Texture Data Object References=Make Array("[0]"=@n14, "[1]"=@n12, "[2]"=@n1.TextureData, "[3]"=@n13)
     Option 3:Array of Building Texture Data Object References=Make Array("[0]"=@n3, "[1]"=@n5, "[2]"=@n4, "[3]"=@n1.TextureData)
     Index:int="Clamp (Integer)"(Value=@n1.Index, Min=0, Max=3) -> (ReturnValue=0)
    ) -> (ReturnValue:Array of Building Texture Data Object References)
   )
  ]}
]
variables [
 AllStaticMeshActorDescs {type="array<ActorDesc>", default []}
 AllStaticMeshes {type="array<StaticMeshActor>", default []}
 Class {type=Class, default=null}
 FilterByStaticMeshAssetName {type=boolean, default=true}
 FilteredStaticMeshes {type="array<StaticMeshActor>", default []}
 Guids {type="array<Guid>", default []}
 Index {type=integer}
 LevelName {type=string, default=Athena_Terrain_S02}
 MeshNameFilter {type="array<string>", default [cliff,elevation,cave,shoreline,slope]}
 New Building Texture Data {type=BuildingTextureData, default=null}
 NewVar {
  type=ActorDesc
  default {
   name=None
   class=""
   actorPackage=None
   actorPath=None
   bActorIsEditorOnly=false
   bIsSpatiallyLoaded=false
   folderPath=None
   guid=#1
   label=None
   nativeClass=null
   runtimeGrid=None
   bounds {isValid=false, max=<0 0 0>, min=<0 0 0>}
   dataLayerAssets []
  }
 }
 NoAssetAssociatedList {type="array<Actor>"}
 NoMaterialAssociatedList {type="array<Actor>"}
 Path {type=string, default=/FigmentEnvironment/Blueprints/Cliffs/}
 ReplaceTD {type=boolean}
 Target Array {type="array<string>", default []}
 TD_Path {type=string, default=""}
 TextureDataDirectory {type=string, default=/FigmentEnvironment/Blueprints/Cliffs/Materials/TextureData}
 UniqueMeshes {type="array<StaticMesh>", default []}
]
widgets {
 info {rootWidgetClass=&CanvasPanel, widgetCount=16}
 widgets [
  CanvasPanel_15 {widgetClassPath=&CanvasPanel, widget=CanvasPanel_15}
  EditorUtilityButton_24 {
   widgetClassPath=&EditorUtilityButton
   parent=CanvasPanel_15
   widget=EditorUtilityButton_24
   slot {
    class=CanvasPanelSlot
    properties {layoutData {alignment=<0 0>, anchors {maximum=<0 0>, minimum=<0 0>}, offsets=<ltrb 20 20 298.8655 51.3547>}}
   }
  }
  TextBlock_22 {
   widgetClassPath=&TextBlock
   parent=EditorUtilityButton_24
   widget=TextBlock_22
   properties {text=Get Static Meshes}
   slot {class=ButtonSlot}
  }
  EditorUtilityButton_2 {
   widgetClassPath=&EditorUtilityButton
   parent=CanvasPanel_15
   widget=EditorUtilityButton_2
   slot {
    class=CanvasPanelSlot
    properties {layoutData {alignment=<0 0>, anchors {maximum=<0 0>, minimum=<0 0>}, offsets=<ltrb 20 92 298.8655 51.3547>}}
   }
  }
  TextBlock_2 {
   widgetClassPath=&TextBlock
   parent=EditorUtilityButton_2
   widget=TextBlock_2
   properties {text=Gather Unique}
   slot {class=ButtonSlot}
  }
  EditorUtilityButton_5 {
   widgetClassPath=&EditorUtilityButton
   parent=CanvasPanel_15
   widget=EditorUtilityButton_5
   slot {
    class=CanvasPanelSlot
    properties {layoutData {alignment=<0 0>, anchors {maximum=<0 0>, minimum=<0 0>}, offsets=<ltrb 20 160 298.8655 51.3547>}}
   }
  }
  TextBlock_5 {
   widgetClassPath=&TextBlock
   parent=EditorUtilityButton_5
   widget=TextBlock_5
   properties {text=Get property type}
   slot {class=ButtonSlot}
  }
  EditorUtilityButton_4 {
   widgetClassPath=&EditorUtilityButton
   parent=CanvasPanel_15
   widget=EditorUtilityButton_4
   slot {
    class=CanvasPanelSlot
    properties {layoutData {alignment=<0 0>, anchors {maximum=<0 0>, minimum=<0 0>}, offsets=<ltrb 352 92 298.8655 51.3547>}}
   }
  }
  TextBlock_4 {
   widgetClassPath=&TextBlock
   parent=EditorUtilityButton_4
   widget=TextBlock_4
   properties {text=Create Assets}
   slot {class=ButtonSlot}
  }
  EditorUtilityButton_3 {
   widgetClassPath=&EditorUtilityButton
   parent=CanvasPanel_15
   widget=EditorUtilityButton_3
   slot {
    class=CanvasPanelSlot
    properties {layoutData {alignment=<0 0>, anchors {maximum=<0 0>, minimum=<0 0>}, offsets=<ltrb 676 96 298.8655 51.3547>}}
   }
  }
  TextBlock_3 {
   widgetClassPath=&TextBlock
   parent=EditorUtilityButton_3
   widget=TextBlock_3
   properties {text=Replace Actors}
   slot {class=ButtonSlot}
  }
  EditorUtilityButton {
   widgetClassPath=&EditorUtilityButton
   parent=CanvasPanel_15
   widget=EditorUtilityButton
   slot {
    class=CanvasPanelSlot
    properties {layoutData {alignment=<0 0>, anchors {maximum=<0 0>, minimum=<0 0>}, offsets=<ltrb 344 20 298.8655 51.3547>}}
   }
  }
  TextBlock {widgetClassPath=&TextBlock, parent=EditorUtilityButton, widget=TextBlock, properties {text=Pin Actors}, slot {class=ButtonSlot}}
  EditorUtilityButton_1 {
   widgetClassPath=&EditorUtilityButton
   parent=CanvasPanel_15
   widget=EditorUtilityButton_1
   slot {
    class=CanvasPanelSlot
    properties {layoutData {alignment=<0 0>, anchors {maximum=<0 0>, minimum=<0 0>}, offsets=<ltrb 668 20 298.8655 51.3547>}}
   }
  }
  TextBlock_1 {
   widgetClassPath=&TextBlock
   parent=EditorUtilityButton_1
   widget=TextBlock_1
   properties {text=Unpin Actors}
   slot {class=ButtonSlot}
  }
  DetailsView_0 {
   widgetClassPath=&DetailsView
   parent=CanvasPanel_15
   widget=DetailsView_0
   slot {
    class=CanvasPanelSlot
    properties {layoutData {alignment=<0 0>, anchors {maximum=<1 1>, minimum=<0 0>}, offsets=<ltrb 40 288.1682 44.1841 44.2042>}}
   }
  }
 ]
}
